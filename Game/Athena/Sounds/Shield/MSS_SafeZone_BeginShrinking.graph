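# One-shot quad MetaSound source that plays the safe zone begin-shrinking cue as a critical in-game UI sound.
# On play, a quad wave player feeds each channel through a master volume multiply to its output, and playback end fires OnFinished.
# A Mobile page swaps in a stereo player with an optimized wave, driving only the front pair.
# Concurrency allows a single instance, and volume follows the game phase UI modulation bus.
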MetaSoundSource /Game/Athena/Sounds/Shield/MSS_SafeZone_BeginShrinking
inputs [
 MasterVolume {type=Float}
 Sound {type=WaveAsset}
 UE.Source.OnPlay {type=Trigger}
]
interfaces [UE.OutputFormat.Quad:1.0, UE.Source.OneShot:1.0, UE.Source:1.1]
outputs [
 "UE.OutputFormat.Quad.Audio:0" {type=Audio}
 "UE.OutputFormat.Quad.Audio:1" {type=Audio}
 "UE.OutputFormat.Quad.Audio:4" {type=Audio}
 "UE.OutputFormat.Quad.Audio:5" {type=Audio}
 UE.Source.OneShot.OnFinished {type=Trigger}
]
pages [
 Default {
  is_build_graph=true
  input_defaults {MasterVolume=3.25, Sound=&/SFX_SafeZone_BeginShrinking_Quad, UE.Source.OnPlay=null}
  nodes [
   WavePlayer_0 = UE.Wave Player:Quad(Wave Asset=^Sound, Play=^UE.Source.OnPlay) {also {version [1,1]}}
   Multiply_0 = UE.Multiply:Audio by Float(AdditionalOperands=^MasterVolume, PrimaryOperand=@WavePlayer_0.Out Front Left)
   Multiply_1 = UE.Multiply:Audio by Float(AdditionalOperands=^MasterVolume, PrimaryOperand=@WavePlayer_0.Out Side Right)
   Multiply_2 = UE.Multiply:Audio by Float(AdditionalOperands=^MasterVolume, PrimaryOperand=@WavePlayer_0.Out Front Right)
   Multiply_3 = UE.Multiply:Audio by Float(AdditionalOperands=^MasterVolume, PrimaryOperand=@WavePlayer_0.Out Side Left)
  ]
  links [
   {dst="@UE.OutputFormat.Quad.Audio:0", src=Multiply_0.Out}
   {dst="@UE.OutputFormat.Quad.Audio:5", src=Multiply_1.Out}
   {dst="@UE.OutputFormat.Quad.Audio:1", src=Multiply_2.Out}
   {dst="@UE.OutputFormat.Quad.Audio:4", src=Multiply_3.Out}
   {dst="@UE.Source.OneShot.OnFinished", src=WavePlayer_0.On Finished}
  ]
 }
 Mobile {
  input_defaults {Sound=&/SFX_SafeZone_BeginShrinking_opt}
  nodes [
   WavePlayer_0 = UE.Wave Player:Stereo(Wave Asset=^Sound, Play=^UE.Source.OnPlay) {also {version [1,1]}}
   Multiply_0 = UE.Multiply:Audio by Float(AdditionalOperands=^MasterVolume, PrimaryOperand=@WavePlayer_0.Out Right)
   Multiply_1 = UE.Multiply:Audio by Float(AdditionalOperands=^MasterVolume, PrimaryOperand=@WavePlayer_0.Out Left)
  ]
  links [
   {dst="@UE.OutputFormat.Quad.Audio:1", src=Multiply_0.Out}
   {dst="@UE.OutputFormat.Quad.Audio:0", src=Multiply_1.Out}
   {dst="@UE.Source.OneShot.OnFinished", src=WavePlayer_0.On Finished}
  ]
 }
]
properties {
 assetImportData=&self:AssetImportData
 bOverrideConcurrency=true
 numChannels=4
 outputFormat=Quad
 sampleRate=48000
 soundClassObject=&Fort_UI_InGame_Critical
 concurrencyOverrides {maxCount=1, resolutionRule=PreventNew}
 modulationSettings {volumeRouting=Union, volumeModulationDestination {bEnableModulation=true, modulators [/Game/Sounds/Modulation/Buses/CB_GamePhaseUI]}}
}
