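# Sound cue that plays the safe zone begin-shrinking sound as a critical in-game UI sound.
# A quality level node selects the Quad wave for its first two levels and the "opt" wave for the last two.
# Overridden concurrency allows one instance and blocks new ones while it plays, with boosted volume and a very large max distance.
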SoundCue /Game/Athena/Sounds/Shield/SafeZone_BeginShrinking_Cue
root=QualityLevel_0
nodes [
 {id=QualityLevel_0, children [WavePlayer_2,WavePlayer_2,WavePlayer_1,WavePlayer_1]}
 {id=WavePlayer_1, properties {soundWaveAssetPtr=&/SFX_SafeZone_BeginShrinking_opt}}
 {id=WavePlayer_2, properties {soundWaveAssetPtr=&/SFX_SafeZone_BeginShrinking_Quad}}
]
properties {
 bOverrideConcurrency=true
 maxDistance=2097152
 soundClassObject=&Fort_UI_InGame_Critical
 volumeMultiplier=3.25
 concurrencyOverrides {maxCount=1, resolutionRule=PreventNew}
}
