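# Sound cue for the Athena safe zone "holding begin" sound, apparently marking the start of the zone's hold phase.
# A single wave player plays SFX_SafeZone_HoldingBegin with boosted volume in the critical in-game UI sound class.
# Overridden concurrency allows one instance at a time and prevents new ones, and the cue sets a very large max distance.
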SoundCue /Game/Athena/Sounds/Shield/SafeZone_HoldingBegin_Cue
root=WavePlayer_0
nodes [{id=WavePlayer_0, properties {soundWaveAssetPtr=&/SFX_SafeZone_HoldingBegin}}]
properties {
 bOverrideConcurrency=true
 maxDistance=2097152
 soundClassObject=&Fort_UI_InGame_Critical
 volumeMultiplier=3.0071
 concurrencyOverrides {maxCount=1, resolutionRule=PreventNew}
}
