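# Sound cue for the looping safe zone holding sound in Athena, stored with the shield sounds.
# A quality level node at the root selects between two looping wave players, with the standard wave on the first two levels and an "_opt" variant on the last two.
# It uses the critical in-game UI sound class, a boosted volume and a very large max distance.
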SoundCue /Game/Athena/Sounds/Shield/SafeZone_HoldingLoop_Cue
root=QualityLevel_0
nodes [
 {id=QualityLevel_0, children [WavePlayer_0,WavePlayer_0,WavePlayer_1,WavePlayer_1]}
 {id=WavePlayer_0, properties {bLooping=true, soundWaveAssetPtr=&/SFX_SafeZone_HoldingLoop}}
 {id=WavePlayer_1, properties {bLooping=true, soundWaveAssetPtr=&/SFX_SafeZone_HoldingLoop_opt}}
]
properties {maxDistance=2097152, soundClassObject=&Fort_UI_InGame_Critical, volumeMultiplier=2.8}
