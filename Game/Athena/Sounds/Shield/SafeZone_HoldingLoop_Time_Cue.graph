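# SafeZone_HoldingLoop_Time_Cue is a sound cue that loops the safe zone holding tick sound, likely during the storm circle's holding phase.
# A single looping wave player feeds the output directly, with no randomization or modulation.
# It uses the critical in-game UI sound class, a heavily boosted volume multiplier and a very large max distance.
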SoundCue /Game/Athena/Sounds/Shield/SafeZone_HoldingLoop_Time_Cue
root=WavePlayer_1
nodes [{id=WavePlayer_1, properties {bLooping=true, soundWaveAssetPtr=&/SFX_SafeZone_HoldingLoop_Tick}}]
properties {maxDistance=2097152, soundClassObject=&Fort_UI_InGame_Critical, volumeMultiplier=9}
