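# Sound cue for the looping storm safe-zone holding timer, playing the Apollo storm countdown-from-10 audio.
# A single looping wave player feeds the root directly, with no mixing or modulation nodes.
# It routes through the critical in-game UI sound class with boosted volume and a very large max distance.
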SoundCue /Game/Athena/Sounds/Shield/SafeZone_HoldingLoop_Time_FN_Cue
root=WavePlayer_2
nodes [{id=WavePlayer_2, properties {bLooping=true, soundWaveAssetPtr=&/Apollo_Storm_Countdown_From10}}]
properties {maxDistance=2097152, soundClassObject=&Fort_UI_InGame_Critical, volumeMultiplier=8}
