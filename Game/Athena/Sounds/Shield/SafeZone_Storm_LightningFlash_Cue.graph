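# Sound cue for the safe zone storm's lightning flash, playing randomized distant thunder through lightning attenuation in the Athena ambient sound class.
# A quality-level node at the root scales variety: the first slot picks from nine thunder waves and the second from three.
# The last two slots share one branch that picks from two optimized waves, with random pitch modulation adding variation to the smaller pool.
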SoundCue /Game/Athena/Sounds/Shield/SafeZone_Storm_LightningFlash_Cue
root=QualityLevel_0
nodes [
 {id=Modulator_0, children [Random_4], properties {pitchMax=1.1, pitchMin=0.8}}
 {id=QualityLevel_0, children [Random_0,Random_2,Modulator_0,Modulator_0]}
 {
  id=Random_0
  children [WavePlayer_1,WavePlayer_2,WavePlayer_3,WavePlayer_4,WavePlayer_5,WavePlayer_6,WavePlayer_7,WavePlayer_8,WavePlayer_9]
  properties {weights [1,1,1,1,1,1,1,1,1]}
 }
 {id=Random_2, children [WavePlayer_0, WavePlayer_10, WavePlayer_11], properties {weights [1,1,1]}}
 {id=Random_4, children [WavePlayer_21, WavePlayer_22], properties {weights [1,1]}}
 {id=WavePlayer_0, properties {soundWaveAssetPtr=/Game/Sounds/Fort_GamePlay_Sounds/Storms/Dist_Thunder_001}}
 {id=WavePlayer_1, properties {soundWaveAssetPtr=&Dist_Thunder_001}}
 {id=WavePlayer_10, properties {soundWaveAssetPtr=/Game/Sounds/Fort_GamePlay_Sounds/Storms/Dist_Thunder_002}}
 {id=WavePlayer_11, properties {soundWaveAssetPtr=/Game/Sounds/Fort_GamePlay_Sounds/Storms/Dist_Thunder_003}}
 {id=WavePlayer_2, properties {soundWaveAssetPtr=&Dist_Thunder_002}}
 {id=WavePlayer_21, properties {soundWaveAssetPtr=/Game/Sounds/Fort_GamePlay_Sounds/Storms/Dist_Thunder_001_opt}}
 {id=WavePlayer_22, properties {soundWaveAssetPtr=/Game/Sounds/Fort_GamePlay_Sounds/Storms/Dist_Thunder_002_opt}}
 {id=WavePlayer_3, properties {soundWaveAssetPtr=&Dist_Thunder_003}}
 {id=WavePlayer_4, properties {soundWaveAssetPtr=/Game/Sounds/Fort_GamePlay_Sounds/Storms/Dist_Thunder_004}}
 {id=WavePlayer_5, properties {soundWaveAssetPtr=/Game/Sounds/Fort_GamePlay_Sounds/Storms/Dist_Thunder_005}}
 {id=WavePlayer_6, properties {soundWaveAssetPtr=/Game/Sounds/Fort_GamePlay_Sounds/Storms/Dist_Thunder_006}}
 {id=WavePlayer_7, properties {soundWaveAssetPtr=/Game/Sounds/Fort_GamePlay_Sounds/Storms/Dist_Thunder_007}}
 {id=WavePlayer_8, properties {soundWaveAssetPtr=/Game/Sounds/Fort_GamePlay_Sounds/Storms/Dist_Thunder_008}}
 {id=WavePlayer_9, properties {soundWaveAssetPtr=/Game/Sounds/Fort_GamePlay_Sounds/Storms/Dist_Thunder_009}}
]
properties {
 attenuationSettings=/Game/Sounds/Attenuation/Fort_SafeZone_Lightning_Athena_Att
 maxDistance=2097152
 soundClassObject=&Fort_Ambient_Athena
 volumeMultiplier=0.208
}
