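# Sound cue for passing through a shield, using safe-zone shield attenuation and the storms sound class.
# A Random node picks one of three equally weighted passthrough waves.
# A boosted volume multiplier raises the level, and overridden concurrency allows one instance per owner, stopping the oldest with a short release when a new one starts.
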SoundCue /Game/Athena/Sounds/Shield/Shield_Passthrough_Cue
root=Random_0
nodes [
 {id=Random_0, children [WavePlayer_1,WavePlayer_2,WavePlayer_3], properties {weights [1,1,1]}}
 {id=WavePlayer_1, properties {soundWaveAssetPtr=&/Shield_Passthrough_01}}
 {id=WavePlayer_2, properties {soundWaveAssetPtr=&/Shield_Passthrough_02}}
 {id=WavePlayer_3, properties {soundWaveAssetPtr=&/Shield_Passthrough_03}}
]
properties {
 attenuationSettings=/Game/Sounds/Attenuation/Fort_SafeZone_Shield_Athena_Att
 bOverrideConcurrency=true
 maxDistance=4000
 soundClassObject=/Game/Sounds/FortSoundClassesAndModes/Fort_Storms_PSM
 volumeMultiplier=3.584
 concurrencyOverrides {bLimitToOwner=true, maxCount=1, resolutionRule=StopOldest, voiceStealReleaseTime=0.5}
}
