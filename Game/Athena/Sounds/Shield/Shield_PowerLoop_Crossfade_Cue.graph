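# Looping Sound Cue for the Athena safe zone shield power hum, played under the storms sound class.
# A "Crossfade" parameter fades from loop 03 to a louder loop 02 as it rises from 0 to 1.
# Loop 03 uses the shared safe zone shield attenuation asset, while loop 02 overrides its own with natural-sound falloff and listener focus.
# Quality-level nodes in each branch select between standard and "_opt" waves.
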SoundCue /Game/Athena/Sounds/Shield/Shield_PowerLoop_Crossfade_Cue
root=ParamCrossFade_2
nodes [
 {
  id=Attenuation_0
  children [QualityLevel_3]
  properties {
   bOverrideAttenuation=true
   attenuationOverrides {
    bEnableFocusInterpolation=true
    bEnableListenerFocus=true
    distanceAlgorithm=NaturalSound
    falloffDistance=6000
    focusAzimuth=25
    focusVolumeAttenuation=1.25
    nonFocusAzimuth=35
    reverbDistanceMax=1024
    reverbDistanceMin=0
    reverbWetLevelMax=0
    reverbWetLevelMin=1
    stereoSpread=1024
   }
  }
 }
 {id=Attenuation_2, children [Mixer_1], properties {attenuationSettings=/Game/Sounds/Attenuation/Fort_SafeZone_Shield_Athena_Att}}
 {id=Mixer_1, children [QualityLevel_1], properties {inputVolume [1]}}
 {
  id=ParamCrossFade_2
  children [Attenuation_2, Attenuation_0]
  properties {
   paramName=Crossfade
   crossFadeInput [
    {fadeInDistanceEnd=0, fadeInDistanceStart=0, fadeOutDistanceEnd=1, fadeOutDistanceStart=0, volume=1}
    {fadeInDistanceEnd=1, fadeInDistanceStart=0, fadeOutDistanceEnd=100, fadeOutDistanceStart=100, volume=2.5}
   ]
  }
 }
 {id=QualityLevel_1, children [WavePlayer_2,WavePlayer_2,WavePlayer_9,WavePlayer_9]}
 {id=QualityLevel_3, children [WavePlayer_5, WavePlayer_5, WavePlayer_11, WavePlayer_11]}
 {id=WavePlayer_11, properties {bLooping=true, soundWaveAssetPtr=&/Shield_PowerLoop_02_opt}}
 {id=WavePlayer_2, properties {bLooping=true, soundWaveAssetPtr=&/Shield_PowerLoop_03}}
 {id=WavePlayer_5, properties {bLooping=true, soundWaveAssetPtr=&/Shield_PowerLoop_02}}
 {id=WavePlayer_9, properties {bLooping=true, soundWaveAssetPtr=&/Shield_PowerLoop_03_opt}}
]
properties {bDebug=true, maxDistance=6400, priority=10, soundClassObject=/Game/Sounds/FortSoundClassesAndModes/Fort_Storms_PSM, volumeMultiplier=0.752}
