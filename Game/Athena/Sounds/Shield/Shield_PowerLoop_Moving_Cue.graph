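# Looping sound cue for the Athena shield power loop, the "moving" variant, which by its safe zone attenuation and storm sound class appears to belong to the storm wall.
# A quality level node at the root picks the standard looping wave for the first two quality slots and an "opt" version for the last two.
# Playback uses a boosted volume multiplier and the safe zone shield attenuation asset.
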SoundCue /Game/Athena/Sounds/Shield/Shield_PowerLoop_Moving_Cue
root=QualityLevel_0
nodes [
 {id=QualityLevel_0, children [WavePlayer_0,WavePlayer_0,WavePlayer_1,WavePlayer_1]}
 {id=WavePlayer_0, properties {bLooping=true, soundWaveAssetPtr=&/Shield_PowerLoop_02}}
 {id=WavePlayer_1, properties {bLooping=true, soundWaveAssetPtr=&/Shield_PowerLoop_02_opt}}
]
properties {
 attenuationSettings=/Game/Sounds/Attenuation/Fort_SafeZone_Shield_Athena_Att
 maxDistance=4000
 soundClassObject=/Game/Sounds/FortSoundClassesAndModes/Fort_Storms_PSM
 volumeMultiplier=2.078
}
